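# Unlit translucent Niagara particle material for the Void Mask teleport screen effect, drawn without depth testing.
# Cell noise distorts an emissive gradient, combined with radially mapped flame tiles sampled three times at small offsets for RGB split, then hue-shifted or solid-tinted.
# A screen-space radial gradient driven by dynamic Radius builds the edge alpha, alongside energy opacity and refraction terms.
# Distance and camera-facing tests build a huge vertex offset that hides the effect.
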Material /DemonMaskVoid/Assets/VFX/CameraFX/M_VoidMask_Teleport_Screen
graph {
 properties {
  bDisableDepthTest=true
  bUsedWithNiagaraMeshParticles=true
  bUsedWithNiagaraSprites=true
  blendMode=BLEND_Translucent
  pixelDepthOffsetMode=PDOM_Legacy
  refractionMethod=RM_IndexOfRefraction
  shadingModel=MSM_Unlit
  usedShadingModels=Unlit
 }
 nodes [
  n1 = WorldPosition()
  n2 = Add(A=CameraPositionWS(), B=@n1.XYZ)
  n3 = Multiply(A=CameraPositionWS(), B=VectorParameter() {parameterName=OffsetMultiplier, defaultValue=<rgba 1 1 1 0>}.RGB)
  n4 = Add(
   A=Multiply(A=ActorPositionWS(), B=VectorParameter() {parameterName=OffsetMultiplier, defaultValue=<rgba 1 1 1 0>}.RGB)
   B=VectorParameter() {parameterName=Offset, defaultValue=<rgba 0 0 5000 0>}.RGB
  )
  n5 = DynamicParameter() {parameterIndex=3, defaultValue=<rgba 1 1 1.25 1>, paramNames [FadeOutDistanceMult, Whiteout, Radius, Param4]}
  n6 = Divide(A=Distance(A=@n3, B=@n4), B=Multiply(A=@n5.FadeOutDistanceMult, B=ScalarParameter() {defaultValue=150000, parameterName=FadeDistance}))
  n7 = Floor(None=Multiply(A=Saturate(None=@n6)) {constB=1.01})
  n8 = Normalize(VectorInput=Subtract(A=@n3, B=@n4))
  n9 = Transform(None=Constant3Vector() {constant=<rgba 0 0 1 0>}) {transformSourceType=TRANSFORMSOURCE_Camera}
  n10 = Multiply(A=OneMinus(None=DotProduct(A=@n8, B=@n9)), B=ScalarParameter() {defaultValue=3, parameterName=PopOutThreshold}) {constB=2}
  n11 = OneMinus(None=Saturate(None=Floor(None=@n10)))
  Add(
   A=Subtract(A=Subtract(A=@n2, B=@n1.XYZ), B=ActorPositionWS())
   B=Multiply(A=Add(A=@n7, B=@n11), B=Constant3Vector() {constant=<rgba 0 0 -54654566400 0>})
  )
  n12 = ScalarParameter() {group=EmissiveNoise, parameterName=EmissiveNoise Rotation}
  n13 = MaterialFunctionCall(
   UVs=Panner(Coordinate=MaterialFunctionCall() {materialFunction=&VectorToRadialValue}.Radial Coordinates) {speedY=1}
   "Rotation Angle (0-1)"=Divide(A=Floor(None=@n12)) {constB=360}
  ) {materialFunction=&CustomRotator}
  n14 = VectorParameter() {group=EmissiveNoise, parameterName=EmissiveNoise UV, defaultValue=<rgba 1 1 0.1 0.1>}
  n15 = AppendVector(A=@n14.R, B=@n14.R)
  n16 = AppendVector(A=@n14.B, B=@n14.B)
  n17 = Time()
  n18 = TextureSampleParameter2D(UVs=Add(A=Multiply(A=@n13.Rotated Values, B=@n15), B=Multiply(A=@n16, B=@n17))) {
   group=EmissiveNoise
   parameterName=EmissiveNoise Tex
   samplerType=SAMPLERTYPE_Grayscale
   texture=/Game/Weapons/FORT_Melee/Pickaxe_Cube_Ninja_Male/FX/T_CellNoise
  }
  n19 = DynamicParameter() {parameterIndex=1, defaultValue=<rgba 0 1 1 1>, paramNames [UV Offset, Noise, Ref, Tile Intensity]}
  n20 = Multiply(A=ScalarParameter() {defaultValue=1, group=EmissiveNoise, parameterName=EmissiveNoise Intensity}, B=@n19.Noise)
  n21 = Multiply(A=OneMinus(None=@n18.R), B=@n20)
  n22 = Subtract(A=@n21, B=Multiply(A=@n18.R, B=@n20))
  n23 = Constant2Vector() {g=0.5}
  n24 = Add(A=Multiply(A=Constant2Vector() {g=0.1}, B=Multiply(A=@n21, B=@n22)), B=@n23)
  n25 = TextureCoordinate()
  n26 = TextureSampleParameter2D!(
   UVs=Multiply(A=Add(A=@n24, B=@n23), B=Constant2Vector() {g=0.5, r=3})
   "DDX(UVs)"=DDX(Value=@n25)
   "DDY(UVs)"=DDY(Value=@n25)
   Apply View MipBias
  ) {
   group=Emissive
   mipValueMode=TMVM_Derivative
   parameterName=Emissive Tex
   texture=/Game/Weapons/FORT_Melee/Pickaxe_Cube_Ninja_Male/FX/T_Offset_Gradient_01
  }
  n27 = Multiply(
   A=MaterialFunctionCall!(TexCoord=@n24, ClampChannel=Constant2Vector() {g=1}) {materialFunction=/Game/Effects/Fort_Effects/MaterialFunctions/MF_FX_TilingMethod}.Result
   B=ChannelMaskParameter(None=@n26.RGBA) {group=Emissive, maskChannel=Blue, parameterName=Emissive Channel, defaultValue=<rgba 0 0 1 0>}.R
  )
  n28 = ComponentMask(None=TextureCoordinate()) {g=true}
  n29 = OneMinus(None=@n28)
  n30 = Multiply(A=Multiply(A=Multiply(A=@n28, B=@n28), B=Multiply(A=@n29, B=@n29))) {constB=10}
  n31 = Multiply(
   A=Multiply(A=@n27, B=ScalarParameter() {defaultValue=1, group=Emissive, parameterName=Emissive Mask})
   B=Multiply(A=Saturate(None=@n30), B=ScalarParameter() {defaultValue=5, parameterName=Mask Multiply}) {constB=10}
  )
  n32 = DynamicParameter() {parameterIndex=1, defaultValue=<rgba 0 1 1 1>, paramNames [UV Offset, Noise, Ref, Tile Intensity]}
  n33 = Multiply(A=@n31, B=Multiply(A=ScalarParameter() {defaultValue=1, parameterName=Ref Intensity}, B=@n32.Ref))
  Add(A=@n33) {constB=0.1}
  n34 = MaterialFunctionCall() {materialFunction=&VectorToRadialValue}
  n35 = MaterialFunctionCall(Float2=@n34.Radial Coordinates) {materialFunction=&BreakOutFloat2Components}
  n36 = AppendVector(A=@n35.R, B=OneMinus(None=@n35.G))
  n37 = ScalarParameter() {parameterName=Invert V}
  n38 = VectorParameter() {group=Emissive, parameterName=Emissive UV, defaultValue=<rgba 1 1 0 0>}
  n39 = DynamicParameter() {parameterIndex=1, defaultValue=<rgba 0 1 1 1>, paramNames [UV Offset, Noise, Ref, Tile Intensity]}
  Add(
   A=Multiply(A=LinearInterpolate(A=@n34.Radial Coordinates, B=@n36, Alpha=@n37), B=AppendVector(A=@n38.R, B=@n38.R))
   B=AppendVector(A=@n38.B, B=StaticSwitchParameter(True=@n39.UV Offset, False=@n38.A) {group=Emissive, parameterName=Use UV Dynamic})
  )
  n40 = ParticleColor()
  n41 = ScalarParameter() {parameterName=HueShift}
  n42 = ComponentMask(None=Multiply(A=MaterialFunctionCall() {materialFunction=&VectorToRadialValue}.Radial Coordinates)) {g=true}
  n43 = MaterialFunctionCall() {materialFunction=&VectorToRadialValue}
  n44 = VectorParameter() {group=TileNoise, parameterName=TileNoise UV, defaultValue=<rgba 2 1 0 0>}
  n45 = AppendVector(A=@n44.R, B=@n44.R)
  n46 = AppendVector(A=@n44.B, B=@n44.B)
  n47 = Time()
  n48 = TextureCoordinate()
  n49 = TextureSampleParameter2D(UVs=Add(A=Multiply(A=@n43.Radial Coordinates, B=@n45), B=Multiply(A=@n46, B=@n47)), "DDX(UVs)"=DDX(Value=@n48), "DDY(UVs)"=DDY(Value=@n48)) {
   group=TileNoise
   mipValueMode=TMVM_Derivative
   parameterName=TileNoise Tex
   samplerType=SAMPLERTYPE_Grayscale
   texture=&LowResBlurredNoise
  }
  n50 = Panner(Coordinate=MaterialFunctionCall() {materialFunction=&VectorToRadialValue}.Radial Coordinates) {speedX=0.01, speedY=0.25}
  n51 = VectorParameter() {group=Tile, parameterName=Tile UV, defaultValue=<rgba 7 1 0 0>}
  n52 = AppendVector(A=@n51.R, B=@n51.R)
  n53 = AppendVector(A=@n51.B, B=@n51.B)
  n54 = Time()
  n55 = Add(
   A=Multiply(A=@n49.R, B=ScalarParameter() {defaultValue=0.0175, group=TileNoise, parameterName=TileNoise Intensity})
   B=Add(A=Multiply(A=@n50, B=@n52), B=Multiply(A=@n53, B=@n54))
  )
  n56 = Multiply(A=ScalarParameter() {defaultValue=1, parameterName=Twitch}) {constB=0.001}
  n57 = TextureSampleParameter2D!(UVs=Add(A=@n55, B=@n56), Apply View MipBias) {
   group=Tile
   parameterName=Tile Tex
   samplerType=SAMPLERTYPE_LinearGrayscale
   texture=/Game/Weapons/FORT_Melee/Pickaxe_Cube_Ninja_Male/FX/T_FluidFlameTile
  }
  n58 = TextureSampleParameter2D(UVs=@n55) {
   group=Tile
   parameterName=Tile Tex
   samplerType=SAMPLERTYPE_LinearGrayscale
   texture=&T_FluidFlameTile
  }
  n59 = TextureSampleParameter2D(UVs=Subtract(A=@n55, B=@n56)) {
   group=Tile
   parameterName=Tile Tex
   samplerType=SAMPLERTYPE_LinearGrayscale
   texture=&T_FluidFlameTile
  }
  n60 = MaterialFunctionCall(X=@n57.R, Y=@n58.R, Z=@n59.R) {materialFunction=&MakeFloat3}
  n61 = VectorParameter() {group=Tile, parameterName=Tile Refine, defaultValue=<rgba 1 1 1 0.5>}
  n62 = LinearInterpolate(A=Multiply(A=Power(Base=@n60.Result, Exp=@n61.R), B=@n61.G), B=Multiply(A=@n60.Result, B=@n61.B), Alpha=@n61.A)
  n63 = ScalarParameter() {defaultValue=1, parameterName=Saturation}
  n64 = VectorParameter() {group=Emissive, parameterName=Emissive Refine, defaultValue=<rgba 1 1 1 0.5>}
  n65 = Multiply(A=Power(Base=@n27, Exp=@n64.R), B=@n64.G)
  n66 = Multiply(A=@n27, B=@n64.B)
  n67 = Multiply(
   A=StaticSwitchParameter(
    True=VectorParameter() {parameterName=SolidColor, defaultValue=<rgba 0 0 0 1>}.RGB
    False=MaterialFunctionCall(Hue Shift Percentage=@n41, Texture=@n42) {materialFunction=&HueShift}.Result
   ) {parameterName=UseSolidColor}
   B=Add(A=Desaturation(None=@n62, Fraction=@n63), B=LinearInterpolate(A=@n65, B=@n66, Alpha=@n64.A))
  )
  Add(A=@n40.RGB, B=@n67)
  Add(A=@n54, B=LinearInterpolate(Alpha=ParticleRandom()) {constB=10})
  Add(A=@n17, B=LinearInterpolate(Alpha=ParticleRandom()) {constB=10})
  Add(A=@n47, B=LinearInterpolate(Alpha=ParticleRandom()) {constB=10})
  Constant3Vector() {constant=<rgba 0.0156 0.0156 0.0156 0>}
  Constant3Vector() {constant=<rgba 0.4747 0 0.8073 1>}
  Constant() {r=1.25}
  Constant() {r=1}
  Constant() {r=1}
  LinearInterpolate(Alpha=Saturate(None=@n5.Whiteout))
  n68 = Multiply(A=@n67, B=@n40.RGB)
  n69 = Desaturation(None=@n68)
  LinearInterpolate(A=@n67, B=VectorParameter() {parameterName=DarkEnergyColor_A, defaultValue=<rgba 0.1302 0.1302 0.1302 1>}.RGB, Alpha=@n69)
  LinearInterpolate(A=@n23, B=Constant2Vector(), Alpha=ParticleRandom())
  MaterialFunctionCall() {materialFunction=&VectorToRadialValue}
  n70 = ScreenPosition()
  n71 = Power(Base=OneMinus(None=MaterialFunctionCall(UVs=@n70.ViewportUV, Radius=@n5.Radius) {materialFunction=&RadialGradientExponential}.RadialGradientExponential)) {constExponent=5}
  n72 = Saturate(None=Multiply(A=@n71) {constB=50})
  Multiply(A=@n5.Whiteout, B=@n72)
  n73 = Multiply(A=@n40.A, B=@n31)
  n74 = DepthFade(Opacity=@n73, FadeDistance=ScalarParameter() {parameterName=DepthFade})
  n75 = Multiply(A=ShadingPathSwitch(Default=@n74, Deferred=@n74, Forward=@n73, Mobile=@n73), B=Power(Base=VertexColor().R) {constExponent=4})
  Multiply(A=@n75, B=OneMinus(None=Fresnel()))
  Multiply(A=TextureCoordinate(), B=Constant2Vector() {g=1, r=2})
  NamedRerouteDeclaration(None=@n72) {name=Alpha, nodeColor=<rgba 1 0 0.6797 1>}
  n76 = MaterialFunctionCall(Emissive=@n68) {materialFunction=&MF_FX_EmissiveAdjustments}
  NamedRerouteDeclaration(None=@n76.Result) {name=EnergyEmissiveColor, nodeColor=<rgba 0 1 0.9531 1>}
  n77 = Multiply(A=Multiply(A=NamedRerouteUsage()), B=@n33)
  NamedRerouteDeclaration(None=LinearInterpolate(Alpha=Saturate(None=@n77)) {constA=1, constB=1.15}) {name=EnergyIOR, nodeColor=<rgba 0 0.5078 1 1>}
  NamedRerouteDeclaration(None=@n69) {name=EnergyOpacity, nodeColor=<rgba 1 0 0.4688 1>}
  NamedRerouteDeclaration(None=@n71) {name=grad, nodeColor=<rgba 0.8281 0 1 1>}
  NamedRerouteUsage()
  NamedRerouteUsage()
  n78 = Constant() {r=0.5}
  n79 = ScalarParameter() {defaultValue=1, parameterName=Radius}
  n80 = Power(
   Base=SmoothStep(Value=SphereMask(A=@n70.ViewportUV, B=@n78, Radius=@n79) {attenuationRadius=0.8, hardnessPercent=0})
   Exp=ScalarParameter() {defaultValue=0.5, parameterName=Pow}
  )
  n81 = ScalarParameter() {defaultValue=1, parameterName=Contribution}
  OneMinus(None=Saturate(None=LinearInterpolate(B=@n80, Alpha=@n81)))
  n82 = NamedRerouteUsage()
  Power(Base=@n82) {constExponent=8}
  Reroute(None=MaterialFunctionCall(Emissive=NamedRerouteUsage()) {materialFunction=&MF_FX_EmissiveAdjustments}.Result)
  Reroute(None=@n76.Result) {desc=Emissive color}
  n83 = Multiply(A=Constant() {r=1}, B=NamedRerouteUsage())
  Saturate(None=Multiply(A=ParticleColor().A, B=Multiply(A=@n83, B=@n82)))
  Saturate(None=DepthFade(Opacity=@n75, FadeDistance=ScalarParameter() {defaultValue=10, parameterName=Depth Fade Distance}) {fadeDistanceDefault=200})
  n84 = OneMinus(None=MaterialFunctionCall(UVs=ScreenPosition().ViewportUV) {materialFunction=&RadialGradientExponential}.RadialGradientExponential)
  Saturate(None=Multiply(A=Power(Base=@n84) {constExponent=3}))
  ScreenPosition()
  ScreenPosition()
  TextureCoordinate()
  TextureCoordinate() {coordinateIndex=1}
  TextureCoordinate()
 ]
}
