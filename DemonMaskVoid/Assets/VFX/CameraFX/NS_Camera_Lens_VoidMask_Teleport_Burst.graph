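# Niagara camera lens effect for the Void Mask teleport, playing a short one-shot purple burst.
# A single emitter bursts one particle into two mesh renderers.
# Curves shape its brightness and a quick alpha rise and fade, with User.GlobalOpacity scaling alpha.
# Dynamic material parameters drive the look, alongside derived dark color and parameter attributes that suggest a darkened second layer.
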NiagaraSystem /DemonMaskVoid/Assets/VFX/CameraFX/NS_Camera_Lens_VoidMask_Teleport_Burst
dependencies {usedDataInterfaces [&NiagaraDataInterfaceCurve], usedDynamicInputs [], usedModules [&SystemState]}
emitters {Debuff_Energy {
  emitterUpdateScript [
   &EmitterState()
   &SpawnBurst_Instantaneous(Spawn Count=1, Use Spawn Probability=false)
  ]
  particleSpawnScript [
   &CameraOffset(Camera Offset Amount=6) {enabled=false}
   &InitializeParticle(
    Color=<rgba 0.5667 0 2 3>
    Lifetime=1.5
    Mass Mode=::NewEnumerator1
    Mesh Scale=<100 100 100>
    Mesh Scale Mode=::NewEnumerator4
    Position Offset=<-250 0 0>
    Sprite Size Mode=::NewEnumerator4
    UsePositionOffset=false
    Write Material Random=true
   ) {module=InitializeParticle001}
  ]
  particleUpdateScript [
   &ParticleState()
   &SolveForcesAndVelocity(Clamp Velocity=false, Limit Acceleration=false)
   &ScaleColor(
    Scale Alpha=&Multiply_Float(A=1, B=$User.GlobalOpacity)
    Scale RGB=&VectorFromFloat(Value=&Lerp_Float(
      A=1
      B=0.01
      Alpha=&FloatFromCurve(FloatCurve=DataInterface() {propertyValues {Curve {keys [{interpMode=RCIM_Cubic, tangentMode=RCTM_User, time=0.2, value=0}, {interpMode=RCIM_Cubic, tangentMode=RCTM_User, time=1, value=1}]}}})
     ))
    ScaleA=true
    ScaleRGB=true
   )
   &ScaleColor(
    Scale Alpha=&FloatFromCurve(
     FloatCurve=DataInterface() {propertyValues {Curve {keys [
         <at 0 0 RCIM_Cubic>
         {arriveTangent=-2.3877, interpMode=RCIM_Cubic, leaveTangent=-2.3877, tangentMode=RCTM_User, time=0.23, value=1}
         {tangentMode=RCTM_User, time=1, value=0}
        ]}}}
     Scale Curve=2
    )
    Scale RGB=&VectorFromFloat(Value=&FloatFromCurve(
      FloatCurve=DataInterface() {propertyValues {Curve {keys [{arriveTangent=0.9111, time=0.2, value=0.1}, {arriveTangent=-0.0405, leaveTangent=-0.0405, time=1, value=0.1}]}}}
      Scale Curve=6
     ))
    ScaleA=true
    ScaleRGB=true
   ) {module=ScaleColor002}
   &DynamicMaterialParameters(
    Index 1 Param 1=0.5
    Index 1 Param 2=0.2
    Index 1 Param 3=1
    Index 3 Param 1=0.9
    Index 3 Param 2=1
    Index 3 Param 3=1
    Module.Index1Param0Write=true
    Module.Index1Param1Write=true
    Module.Index1Param2Write=true
    Module.Index1Param3Write=false
    Module.Index3Param0Write=true
    Module.Index3Param1Write=true
    Module.Index3Param2Write=true
    Module.Index3Param3Write=false
    Write Parameter Index 0=false
    Write Parameter Index 1=true
    Write Parameter Index 3=true
   ) {module=DynamicMaterialParameters001}
   /SetVariables(
    Particles.ColorDark:LinearColor=&MultiplyLinearColor(Linear Color=$Particles.Color, Scale Factor=<0.01 0.01 0.01 80>)
    Particles.DynamicMaterialParameter3_Dark:Vector4f=&MakeVector4(X=1, Y=1, Z=0.78, W=0)
   ) {isSetParameters=true}
  ]
  rendererClasses [&NiagaraMeshRendererProperties]
  renderers [
   {rendererClass=&NiagaraMeshRendererProperties, rendererIndex=0}
   {rendererClass=&NiagaraMeshRendererProperties, rendererIndex=1}
  ]
 }}
summary {
 systemName=NS_Camera_Lens_VoidMask_Teleport_Burst
 emitters [{emitterName=Debuff_Energy, rendererClasses [&NiagaraMeshRendererProperties]}]
 userVariables [User.GlobalOpacity {type=NiagaraFloat, defaultValue=1, description=""}]
}
